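# Animation Blueprint for the F_LRG_Billy body that drives corrective helper bones at the thighs and right upper arm.
# Each update maps thigh and upperarm_r swing angles from the ref pose to clamped alphas.
# The AnimGraph copies the attached parent's pose, then applies alpha-weighted additive parent-space Modify Bone offsets.
# Back-thigh bones use the front-thigh alphas, leaving their own variables unread, and the right back-thigh offset is zero.
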AnimBlueprint /BRCosmetics/Characters/Player/Female/Large/Bodies/F_LRG_Billy/Meshes/F_LRG_Billy_AnimBP
target_skeleton=&/F_LRG_Billy_Skeleton
graphs [
 AnimGraph {nodes [
   n1 = Get extra_front_thigh_l_alpha() -> (extra_front_thigh_l_alpha:double)
   n2 = "Transform (Modify) Bone - Bone: extra_front_thigh_l"(
    ComponentPose=Local To Component(LocalPose=Copy Pose From Mesh() {bUseAttachedParent=true})
    Alpha=@n1
    Translation="0,25,0"
    Rotation="0,0,0"
    Scale="1,1,1"
   ) {
    rotationMode=BMM_Additive
    rotationSpace=BCS_ParentBoneSpace
    translationMode=BMM_Additive
    translationSpace=BCS_ParentBoneSpace
    alphaBoolBlend {blendOption=Linear}
    boneToModify {boneName=extra_front_thigh_l}
    translation {y=25}
   }
   n3 = Get extra_front_thigh_r_alpha() -> (extra_front_thigh_r_alpha:double)
   n4 = "Transform (Modify) Bone - Bone: extra_front_thigh_r"(ComponentPose=@n2, Alpha=@n3, Translation="0,-25,0", Rotation="0,0,0", Scale="1,1,1") {
    rotationMode=BMM_Additive
    rotationSpace=BCS_ParentBoneSpace
    translationMode=BMM_Additive
    translationSpace=BCS_ParentBoneSpace
    alphaBoolBlend {blendOption=Linear}
    boneToModify {boneName=extra_front_thigh_r}
    translation {y=-25}
   }
   n5 = "Transform (Modify) Bone - Bone: extra_back_thigh_l"(ComponentPose=@n4, Alpha=@n1, Translation="27.964,-5.816,24.888", Rotation="0,0,0", Scale="1,1,1") {
    rotationMode=BMM_Additive
    rotationSpace=BCS_ParentBoneSpace
    translationMode=BMM_Additive
    translationSpace=BCS_ParentBoneSpace
    alphaBoolBlend {blendOption=Linear}
    boneToModify {boneName=extra_back_thigh_l}
    translation=<27.964 -5.816 24.888>
   }
   n6 = "Transform (Modify) Bone - Bone: extra_back_thigh_r"(ComponentPose=@n5, Alpha=@n3, Translation="0,0,0", Rotation="0,0,0", Scale="1,1,1") {
    rotationMode=BMM_Additive
    rotationSpace=BCS_ParentBoneSpace
    translationMode=BMM_Additive
    translationSpace=BCS_ParentBoneSpace
    alphaBoolBlend {blendOption=Linear}
    boneToModify {boneName=extra_back_thigh_r}
   }
   n7 = "Transform (Modify) Bone - Bone: extra_upperarm_r"(
    ComponentPose=@n6
    Alpha=Get extra_upperarm_r_alpha() -> (extra_upperarm_r_alpha:double)
    Translation="-1.770,12.147,3.374"
    Rotation="0,0,0"
    Scale="1,1,1"
   ) {
    rotationMode=BMM_Additive
    rotationSpace=BCS_ParentBoneSpace
    translationMode=BMM_Additive
    translationSpace=BCS_ParentBoneSpace
    alphaBoolBlend {blendOption=Linear}
    boneToModify {boneName=extra_upperarm_r}
    translation=<-1.77 12.147 3.374>
   }
   Output Pose(Result=Component To Local(ComponentPose=@n7)) {layerGroup=None}
  ]}
 EventGraph {nodes [
   Event BlueprintUpdateAnimation()
   n1 = GetOwnerSkeletalMeshComponent()
   n2 = MapRangeClamped(Value=GetTwistAndSwingAngleOfDeltaRotationFromRefPose(self=@n1, BoneName=thigh_l).OutSwingAngle, InRangeA=23, InRangeB=104)
   > Set extra_back_thigh_l_alpha(extra_back_thigh_l_alpha:double=@n2) -> (Output_Get:double)
   n3 = MapRangeClamped(Value=GetTwistAndSwingAngleOfDeltaRotationFromRefPose(self=@n1, BoneName=thigh_r).OutSwingAngle, InRangeA=163, InRangeB=114)
   > Set extra_back_thigh_r_alpha(extra_back_thigh_r_alpha:double=@n3) -> (Output_Get:double)
   > Set extra_front_thigh_l_alpha(extra_front_thigh_l_alpha:double=@n2) -> (Output_Get:double)
   > Set extra_front_thigh_r_alpha(extra_front_thigh_r_alpha:double=@n3) -> (Output_Get:double)
   > Set extra_upperarm_r_alpha(extra_upperarm_r_alpha:double=MapRangeClamped(Value=GetTwistAndSwingAngleOfDeltaRotationFromRefPose(self=@n1, BoneName=upperarm_r).OutSwingAngle, InRangeA=67, InRangeB=123)) -> (Output_Get:double)
  ]}
]
variables [
 Bone Name {type=string, default=front_thigh_l}
 extra_back_thigh_l_alpha {type=number}
 extra_back_thigh_r_alpha {type=number}
 extra_front_thigh_l_alpha {type=number}
 extra_front_thigh_r_alpha {type=number}
 extra_upperarm_r_alpha {type=number}
]
