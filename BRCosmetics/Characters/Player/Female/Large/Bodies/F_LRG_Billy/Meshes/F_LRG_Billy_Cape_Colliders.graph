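# Physics asset for the large female Billy player body that, by its name, supplies the body colliders for her cape.
# It holds one capsule per body on the pelvis, spine, arms, hands, thigh and calf twist bones, and feet, with no bodies for head, neck or cape bones.
# Uniform 45-degree constraints chain spine and arms, while each leg capsule links straight to the pelvis rather than forming a limb chain.
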PhysicsAsset /BRCosmetics/Characters/Player/Female/Large/Bodies/F_LRG_Billy/Meshes/F_LRG_Billy_Cape_Colliders
bodies {
 calf_twist_02_l [{shapeType=Capsule, length=21.507, radius=7.7318, center=<-1.5956 -1.7786 -0.619>, rotation=<rot 0 90.1539 -89.0162>}]
 calf_twist_02_r [{shapeType=Capsule, length=23.5095, radius=7.7009, center=<3.5635 1.8453 0.2892>, rotation=<rot 0 90.1873 -88.9903>}]
 foot_l [{shapeType=Capsule, length=14.5659, radius=6.3882, center=<-4.1516 7.989 -0.5419>, rotation=<rot 0 3.0592 -89.5567>}]
 foot_r [{shapeType=Capsule, length=14.5677, radius=6.3352, center=<4.0351 -7.9436 0.5378>, rotation=<rot 0 2.3014 -89.509>}]
 hand_l [{shapeType=Capsule, length=9.0333, radius=7.5405, center=<7.5534 -2.8307 0.0945>, rotation=<rot 0 86.1792 -62.4028>}]
 hand_r [{shapeType=Capsule, length=9.0327, radius=7.5409, center=<-7.5532 2.8294 -0.0944>, rotation=<rot 0 86.205 -62.3968>}]
 lowerarm_l [{shapeType=Capsule, length=13.4004, radius=6.4169, center=<11.138 1.0841 0.2068>, rotation=<rot 0 90.1682 -87.7942>}]
 lowerarm_r [{shapeType=Capsule, length=13.3997, radius=6.4186, center=<-11.1384 -1.0842 -0.2062>, rotation=<rot 0 90.1642 -87.8089>}]
 pelvis [{shapeType=Capsule, length=18.111, radius=13.273, center=<-0.6321 -1.396 -0.1153>, rotation=<rot -3.8197 -90 -0.6659>}]
 spine_01 [{shapeType=Capsule, length=14.4739, radius=10.9552, center=<8.4444 0.5362 0.099>, rotation=<rot -3.7857 -90 1.1243>}]
 spine_03 [{shapeType=Capsule, length=15.4216, radius=13.225, center=<2.8934 0.9684 -0.044>, rotation=<rot 0.7296 -90 -0.0636>}]
 spine_04 [{shapeType=Capsule, length=16.2854, radius=14.8872, center=<2.8311 1.0983 0.0471>, rotation=<rot -0.4032 -90 -0.1199>}]
 spine_05 [{shapeType=Capsule, length=17.4791, radius=14.2008, center=<3.4029 -0.0423 0.0299>, rotation=<rot 0.2461 -90 0.0048>}]
 thigh_twist_01_l [{shapeType=Capsule, length=21.681, radius=9.9708, center=<5.5924 -0.8745 0.4939>, rotation=<rot 0 -80.1885 -83.4948>}]
 thigh_twist_01_r [{shapeType=Capsule, length=21.724, radius=9.9977, center=<-5.5448 0.8867 -0.5139>, rotation=<rot 0 -79.9621 -83.8763>}]
 upperarm_l [{shapeType=Capsule, length=15.1043, radius=9.0176, center=<11.2473 1.208 -1.6892>, rotation=<rot 0 100.8036 -89.266>}]
 upperarm_r [{shapeType=Capsule, length=15.1071, radius=8.9038, center=<-11.238 -1.1952 1.4036>, rotation=<rot 0 100.7877 -89.4674>}]
}
constraints [
 {bone1Name=spine_01, bone2Name=pelvis, swing1LimitDegrees=45, swing2LimitDegrees=45, twistLimitDegrees=45}
 {bone1Name=spine_03, bone2Name=spine_01, swing1LimitDegrees=45, swing2LimitDegrees=45, twistLimitDegrees=45}
 {bone1Name=spine_04, bone2Name=spine_03, swing1LimitDegrees=45, swing2LimitDegrees=45, twistLimitDegrees=45}
 {bone1Name=spine_05, bone2Name=spine_04, swing1LimitDegrees=45, swing2LimitDegrees=45, twistLimitDegrees=45}
 {bone1Name=upperarm_l, bone2Name=spine_05, swing1LimitDegrees=45, swing2LimitDegrees=45, twistLimitDegrees=45}
 {bone1Name=lowerarm_l, bone2Name=upperarm_l, swing1LimitDegrees=45, swing2LimitDegrees=45, twistLimitDegrees=45}
 {bone1Name=hand_l, bone2Name=lowerarm_l, swing1LimitDegrees=45, swing2LimitDegrees=45, twistLimitDegrees=45}
 {bone1Name=upperarm_r, bone2Name=spine_05, swing1LimitDegrees=45, swing2LimitDegrees=45, twistLimitDegrees=45}
 {bone1Name=lowerarm_r, bone2Name=upperarm_r, swing1LimitDegrees=45, swing2LimitDegrees=45, twistLimitDegrees=45}
 {bone1Name=hand_r, bone2Name=lowerarm_r, swing1LimitDegrees=45, swing2LimitDegrees=45, twistLimitDegrees=45}
 {bone1Name=foot_l, bone2Name=pelvis, swing1LimitDegrees=45, swing2LimitDegrees=45, twistLimitDegrees=45}
 {bone1Name=calf_twist_02_l, bone2Name=pelvis, swing1LimitDegrees=45, swing2LimitDegrees=45, twistLimitDegrees=45}
 {bone1Name=thigh_twist_01_l, bone2Name=pelvis, swing1LimitDegrees=45, swing2LimitDegrees=45, twistLimitDegrees=45}
 {bone1Name=foot_r, bone2Name=pelvis, swing1LimitDegrees=45, swing2LimitDegrees=45, twistLimitDegrees=45}
 {bone1Name=calf_twist_02_r, bone2Name=pelvis, swing1LimitDegrees=45, swing2LimitDegrees=45, twistLimitDegrees=45}
 {bone1Name=thigh_twist_01_r, bone2Name=pelvis, swing1LimitDegrees=45, swing2LimitDegrees=45, twistLimitDegrees=45}
]
