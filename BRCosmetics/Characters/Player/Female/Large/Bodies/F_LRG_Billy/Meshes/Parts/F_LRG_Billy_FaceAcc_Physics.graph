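# Physics asset for F_LRG_Billy's face accessory mesh, giving capsule bodies to the head, both ears and a left-ear tag.
# Each three-bone ear chain locks its root angularly to the head, with outer joints set to narrow swing1 and twist limits and wider swing2 limits, suggesting hinge-like flapping.
# The two-bone tag locks to the left ear tip and leaves only swing1 unlocked at its second joint.
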PhysicsAsset /BRCosmetics/Characters/Player/Female/Large/Bodies/F_LRG_Billy/Meshes/Parts/F_LRG_Billy_FaceAcc_Physics
bodies {
 L_dyn_ear_01_joint [{
   shapeType=Capsule
   length=0.195
   radius=0.5829
   shapeName=L_dyn_ear_01_joint_capsule
   center=<0.0436 -0.088 0.0456>
   rotation=<rot 6.759 -90 89.956>
  }]
 L_dyn_ear_02_joint [{
   shapeType=Capsule
   length=0.9263
   radius=3.4696
   shapeName=L_dyn_ear_02_joint_capsule
   center=<1.6576 0 0.0159>
   rotation=<rot -0.0002 -90 89.4498>
  }]
 L_dyn_ear_03_joint [{
   shapeType=Capsule
   length=1.6993
   radius=3.8786
   shapeName=L_dyn_ear_03_joint_capsule
   center=<3.07 -0.2321 -0.1143>
   rotation=<rot 0.0001 85.8391 -92.7431>
  }]
 L_dyn_tag_01_joint [{
   shapeType=Capsule
   length=1.8839
   radius=1.1156
   shapeName=L_dyn_tag_01_joint_capsule
   center=<1.27 -0.0691 -0.0856>
   rotation=<rot -10.8485 -90 -3.1232>
  }]
 L_dyn_tag_02_joint [{
   shapeType=Capsule
   length=2.295
   radius=0.9118
   shapeName=L_dyn_tag_02_joint_capsule
   center=<1.4034 0.05 -0.0202>
   rotation=<rot 0.2127 -90 -0.0886>
  }]
 R_dyn_ear_01_joint [{
   shapeType=Capsule
   length=0.195
   radius=0.5829
   shapeName=L_dyn_ear_01_joint_capsule
   center=<-0.0436 0.088 -0.0456>
   rotation=<rot 6.759 -90 -90.044>
  }]
 R_dyn_ear_02_joint [{
   shapeType=Capsule
   length=0.9263
   radius=3.4696
   shapeName=L_dyn_ear_02_joint_capsule
   center=<-0.3051 0 -0.0029>
   rotation=<rot -0.0002 -90 -90.5502>
  }]
 R_dyn_ear_03_joint [{
   shapeType=Capsule
   length=1.6993
   radius=3.8786
   shapeName=L_dyn_ear_03_joint_capsule
   center=<-1.3363 0.106 0.031>
   rotation=<rot 0.0001 85.8391 87.2569>
  }]
 head [{shapeType=Capsule, length=5.6637, radius=4.2654, center=<5.9572 3.4984 0.3112>, rotation=<rot -0.1389 -90.0001 0.1838>}]
}
constraints [
 {
  bone1Name=L_dyn_ear_01_joint
  bone2Name=head
  swing1LimitDegrees=45
  swing1Motion=Locked
  swing2LimitDegrees=45
  swing2Motion=Locked
  twistLimitDegrees=45
  twistMotion=Locked
 }
 {bone1Name=L_dyn_ear_02_joint, bone2Name=L_dyn_ear_01_joint, swing1LimitDegrees=0.5, swing2LimitDegrees=60, twistLimitDegrees=2}
 {bone1Name=L_dyn_ear_03_joint, bone2Name=L_dyn_ear_02_joint, swing1LimitDegrees=0.5, swing2LimitDegrees=40, twistLimitDegrees=2}
 {
  bone1Name=R_dyn_ear_01_joint
  bone2Name=head
  swing1LimitDegrees=45
  swing1Motion=Locked
  swing2LimitDegrees=45
  swing2Motion=Locked
  twistLimitDegrees=45
  twistMotion=Locked
 }
 {
  bone1Name=L_dyn_tag_02_joint
  bone2Name=L_dyn_tag_01_joint
  swing1LimitDegrees=20
  swing2LimitDegrees=45
  swing2Motion=Locked
  twistLimitDegrees=45
  twistMotion=Locked
 }
 {
  bone1Name=L_dyn_tag_01_joint
  bone2Name=L_dyn_ear_03_joint
  swing1LimitDegrees=0
  swing1Motion=Locked
  swing2LimitDegrees=45
  swing2Motion=Locked
  twistLimitDegrees=45
  twistMotion=Locked
 }
 {bone1Name=R_dyn_ear_02_joint, bone2Name=R_dyn_ear_01_joint, swing1LimitDegrees=0.5, swing2LimitDegrees=60, twistLimitDegrees=2}
 {bone1Name=R_dyn_ear_03_joint, bone2Name=R_dyn_ear_02_joint, swing1LimitDegrees=0.5, swing2LimitDegrees=40, twistLimitDegrees=2}
]
