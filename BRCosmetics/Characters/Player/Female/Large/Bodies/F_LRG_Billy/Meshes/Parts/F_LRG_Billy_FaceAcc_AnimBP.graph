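# Animation Blueprint for the Billy female-large face accessory, adding physics-driven secondary motion to the part.
# AnimGraph copies the attached parent's pose into a RigidBody node based at neck_01, with variable gravity and blend alpha.
# EventGraph sets skydiving gravity through a deprecated state function and picks alpha by parachute state.
# Otherwise, moving without sprinting or targeting eases alpha down or up without resetting gravity; other states restore default gravity and alpha.
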AnimBlueprint /BRCosmetics/Characters/Player/Female/Large/Bodies/F_LRG_Billy/Meshes/Parts/F_LRG_Billy_FaceAcc_AnimBP
target_skeleton=&/F_LRG_Billy_FaceAcc_Skeleton
graphs [
 AnimGraph {nodes [
   n1 = RigidBody!(
    ComponentPose=Local To Component(LocalPose=Copy Pose From Mesh() {bUseAttachedParent=true})
    Alpha=Get alphaDin() -> (alphaDin:double)
    OverrideWorldGravity:Vector=Get skyDiving() -> (skyDiving:Vector)
    ExternalForce
   ) -> (Pose) {
    bOverrideWorldGravity=true
    overridePhysicsAsset=&/F_LRG_Billy_FaceAcc_Physics
    alphaBoolBlend {blendOption=Linear}
    baseBoneRef {boneName=neck_01}
    componentLinearAccScale=<0.5 0.5 0.5>
    componentLinearVelScale=<0.8 0.8 0.8>
    overrideWorldGravity {z=-980}
    simSpaceSettings {maxAngularAcceleration=500, maxAngularVelocity=3, maxLinearAcceleration=400, maxLinearVelocity=250}
   }
   Output Pose(Result=Component To Local(ComponentPose=@n1)) {layerGroup=None}
  ]}
 EventGraph {nodes [
   n1 = Event BlueprintUpdateAnimation()
   > n2 = Branch(Condition=Get bIsSkyDiving())
   > n3 = Gravity Override State Function DEPRECATED(CharacterPartAnimInstance=Self-Reference(), GameTimeInSeconds=GetGameTimeInSeconds(), GravityOverrideParams=Get Gravity Override Params())
   > Set skyDiving(skyDiving:Vector=@n3) -> (Output_Get:Vector)
   > Set alphaDin(alphaDin:double=SelectFloat(A=0.2, B=0.6, bPickA=Get bIsParachuteOpened())) -> (Output_Get:double)
   n4 = Get bIsSprinting()
   n5 = Get bIsTargeting()
   n6 = Branch(execute=@n2.else, Condition=AND Boolean!(A=NOT Boolean(A=@n4), B=Get bIsMoving2D(), C:bool=NOT Boolean(A=@n5)) -> (ReturnValue))
   > Set alphaDin(alphaDin=FInterpTo(Current=Get alphaDin(), Target=0.4, DeltaTime=@n1.DeltaTimeX, InterpSpeed=2))
   Set skyDiving(execute=@n6.else, skyDiving="0, 0,-980")
   > Set alphaDin(alphaDin=0.6)
  ]}
]
variables [
 alphaDin {type=number}
 Gravity Override Params {type=GravityOverrideParamsStruct}
 R_alphaDin {type=number}
 skyDiving {type=Vector}
]
